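# QA editor utility that verifies items in station and NPC sales loot have entries in the CashPricingItemDef curve table.
# Run force-loads soft table references, tracing sales LootTiers through tier data to package IDs, package calls, then item asset names.
# Each stage fills a deduplicated tracking array for the next, with bool-toggled logging.
# The final item names are compared against curve table row names, printing an error for each missing item.
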EditorUtilityBlueprint /Game/NeverCook/QA/QuestTech_QuestAutomation/EU_BR_ItemSalesVerif
graphs [
 EventGraph {nodes [
   CompareTrackedItemsToCashPricingList() {is=title K2Node_CustomEvent}
   > n1 = For Each Loop(Array=Get PackageItemListTrackingArray() -> (PackageItemListTrackingArray:Array of Strings))
   PrintString(execute=@n1.Completed, InString="Item Sales Verification Complete, check above for error logs.")
   n2 = Branch(
    execute=@n1.LoopBody
    Condition=Contains Item(TargetArray:Array of Strings=Get CashPricingTrackingArray() -> (CashPricingTrackingArray:Array of Strings), ItemToFind:"String (by ref)"=@n1)
   )
   > PrintString(
    InString=Append(A=@n1, B=" is in both the tracked Item List and the CashPricingItemDef table. \\o/")
    bPrintToLog=Get bShowAllLogging() -> (bShowAllLogging:bool)
   )
   PrintString(
    execute=@n2.else
    InString=Append!(A="ERROR: ", B=@n1, C:String=" is in the tracked Item List but not in the CashPricingItemDef table.") -> (ReturnValue)
   )
   Event Run()
   > n3 = ForceLoadDataTableFromSoftObjectPtr!(
    execute
    self:EUO Item Sale Verification Object Reference
    DataTableToLoad:Data Table Soft Object Reference=Get CurrentStationSales() -> (CurrentStationSales:Data Table Soft Object Reference)
   ) -> (then, ReturnValue:Data Table Object Reference)
   > GetTierGroupReferencesFromSalesTables!(execute, self, SalesDataTable:Data Table Object Reference=@n3) -> (then)
   > n4 = ForceLoadDataTableFromSoftObjectPtr(DataTableToLoad=Get CurrentNpcSales() -> (CurrentNpcSales:Data Table Soft Object Reference))
   > GetTierGroupReferencesFromSalesTables(SalesDataTable=@n4)
   n5 = Length(TargetArray:Array of Names=Get TierGroupTrackingArray() -> (TierGroupTrackingArray:Array of Names))
   > PrintString(
    InString=Append!(
     A="Tier Group Num: "
     B="To String (Integer)"(InInt=@n5)
     C=" from "
     D:String=GetDisplayName(Object=@n3)
     E:String=" and "
     F:String=GetDisplayName(Object=@n4)
    ) -> (ReturnValue)
    bPrintToScreen=false
    Key=ItemSaleVerif
   )
   > n6 = ForceLoadDataTableFromSoftObjectPtr(DataTableToLoad=Get LootCurrentSeasonLootTierData() -> (LootCurrentSeasonLootTierData:Data Table Soft Object Reference))
   > GetLootPackageIdFromTierData!(execute, self, TierGroupNames:Array of Names=Get TierGroupTrackingArray(), LootTierDataTable:Data Table Object Reference=@n6) -> (then)
   n7 = Length(TargetArray=Get LootPackageIdTrackingArray() -> (LootPackageIdTrackingArray:Array of Names))
   > PrintString(
    InString=Append!(A="Loot Package Id Num: ", B="To String (Integer)"(InInt=@n7), C=" from ", D=GetDisplayName(Object=@n6)) -> (ReturnValue)
    bPrintToScreen=false
    Key=ItemSaleVerif
   )
   > n8 = ForceLoadDataTableFromSoftObjectPtr(DataTableToLoad=Get LootCurrentSeasonLootPackageData() -> (LootCurrentSeasonLootPackageData:Data Table Soft Object Reference))
   > GetLootPackageCallFromPackageIdData!(execute, self, PackageIdNames:Array of Names=Get LootPackageIdTrackingArray(), LootPackageDataTable:Data Table Object Reference=@n8) -> (then)
   n9 = Length(TargetArray:Array of Strings=Get LootPackageCallTrackingArray() -> (LootPackageCallTrackingArray:Array of Strings))
   > PrintString(
    InString=Append(A="Loot Package Call Num: ", B="To String (Integer)"(InInt=@n9), C=" from ", D=GetDisplayName(Object=@n8))
    bPrintToScreen=false
    Key=ItemSaleVerif
   )
   > n10 = ForceLoadDataTableFromSoftObjectPtr(DataTableToLoad=Get LootCurrentSeasonLootPackageData())
   > GetItemListFromLootPackageCallData!(execute, self, LootPackageCalls:Array of Strings=Get LootPackageCallTrackingArray(), LootPackageDataTable:Data Table Object Reference=@n10) -> (then)
   n11 = Length(TargetArray=Get PackageItemListTrackingArray())
   > PrintString(
    InString=Append(A="Item List Num: ", B="To String (Integer)"(InInt=@n11), C=" from ", D=GetDisplayName(Object=@n10))
    bPrintToScreen=false
    Key=ItemSaleVerif
   )
   > n12 = ForceLoadAssetFromSoftObjectPtr!(
    execute
    self:EUO Item Sale Verification Object Reference
    ObjectToLoad:Soft Object Reference=Get CashPricingItemDef() -> (CashPricingItemDef:Soft Object Reference)
   ) -> (then, ReturnValue:Object Reference)
   > n13 = GetRowNamesOfCurveTable!(execute, self:EUO Item Sale Verification Object Reference, KnownCurveTable:Object Reference=@n12) -> (then, ReturnValue:Array of Strings)
   > Set CashPricingTrackingArray(CashPricingTrackingArray:Array of Strings=@n13) -> (Output_Get:Array of Strings)
   > CompareTrackedItemsToCashPricingList() {is=title K2Node_CallFunction}
  ]}
 GetItemListFromLootPackageCallData {nodes [
   n1 = GetItemListFromLootPackageCallData!() -> (then, LootPackageCalls:Array of Strings, LootPackageDataTable:Data Table Object Reference) {is=title K2Node_FunctionEntry}
   > n2 = GetDataTableRowNames(Table=@n1.LootPackageDataTable)
   > n3 = For Each Loop(Array:Array of Names=@n2) -> (Array Element:Name)
   > n4 = Get Data Table Row!(execute:Exec, DataTable:Data Table Object Reference=@n1.LootPackageDataTable, RowName:Name=@n3) -> (then:Exec, RowNotFound:Exec, ReturnValue:Fort Loot Package Data Structure)
   n5 = Break Fort Loot Package Data!(FortLootPackageData:"Fort Loot Package Data Structure (by ref)"=@n4) -> (
    LootPackageID:Name
    CountRange:Int Vector 2 Structure
    RequiredLootGroupTag:Gameplay Tag Structure
    RequiredTagQuery:Gameplay Tag Query Structure
    ItemDefinition:Fort Item Definition Soft Object Reference
   ) {is=title K2Node_BreakStruct}
   n6 = "To String (Name)"(InName=@n5)
   > n7 = Branch(Condition=Contains Item(TargetArray=@n1, ItemToFind=@n6))
   > n8 = Branch(Condition=IsValidSoftObjectReference(SoftObjectReference=@n5.ItemDefinition))
   n9 = GetSoftPtrAssetName!(Asset:Soft Object Reference=@n5.ItemDefinition) -> (ReturnValue:String)
   > n10 = Add Unique(TargetArray=Get PackageItemListTrackingArray(), NewItem=@n9)
   > Branch(Condition="integer >= integer"(A=@n10))
   > PrintString(
    InString=Append(A="Tracking Item: ", B=@n9, C=" from: ", D=@n6)
    bPrintToScreen=false
    bPrintToLog=OR Boolean!(
     A=Get bShowAllLogging()
     B=Get bShowItemListLogging() -> (bShowItemListLogging:bool)
     C:bool=Get bShowAllSuccessLogging() -> (bShowAllSuccessLogging:bool)
    ) -> (ReturnValue)
   )
   PrintString(
    execute=@n8.else
    InString=Append!(A="ERROR: Item: ", B=@n9, C="does not have a valid  ItemDef, is this ok:") -> (ReturnValue)
    bPrintToScreen=false
    Key=ItemSaleVerif
   )
   n11 = Branch(execute=@n7.else, Condition=IsEmpty(InString=@n9))
   PrintString(
    execute=@n11.else
    InString=Append!(A="Package Id: ", B=@n6, C=" is not tracked in the Loot Package Calls, not tracking: ", D:String=@n9) -> (ReturnValue)
    bPrintToLog=OR Boolean!(A=Get bShowAllLogging(), B=Get bShowItemListLogging()) -> (ReturnValue)
   )
  ]}
 GetLootPackageCallFromPackageIdData {nodes [
   n1 = GetLootPackageCallFromPackageIdData!() -> (then, PackageIdNames:Array of Names, LootPackageDataTable:Data Table Object Reference) {is=title K2Node_FunctionEntry}
   > n2 = GetDataTableRowNames(Table=@n1.LootPackageDataTable)
   > n3 = For Each Loop(Array=@n2)
   Return Node(execute=@n3.Completed)
   > n4 = Get Data Table Row(DataTable=@n1.LootPackageDataTable, RowName=@n3)
   n5 = Break Fort Loot Package Data!(FortLootPackageData=@n4) -> (LootPackageID, CountRange, RequiredLootGroupTag, RequiredTagQuery, LootPackageCall:String)
   > n6 = Branch(Condition=Contains Item(TargetArray:Array of Names=@n1, ItemToFind:"Name (by ref)"=@n5))
   > n7 = Add Unique(TargetArray=Get LootPackageCallTrackingArray(), NewItem=@n5.LootPackageCall)
   > Branch(Condition="integer >= integer"(A=@n7))
   > PrintString(
    InString=Append(A="Tracking Package Call: ", B=@n5.LootPackageCall, C=" from: ", D="To String (Name)"(InName=@n5))
    bPrintToScreen=false
    bPrintToLog=OR Boolean!(A=Get bShowAllLogging(), B=Get bShowPackageCallLogging() -> (bShowPackageCallLogging:bool), C:bool=Get bShowAllSuccessLogging()) -> (ReturnValue)
   )
   PrintString(
    execute=@n6.else
    InString=Append(A="Package Id: ", B="To String (Name)"(InName=@n5), C=" is not being tracked, not tracking: ", D=@n5.LootPackageCall)
    bPrintToScreen=false
    bPrintToLog=OR Boolean!(A=Get bShowAllLogging(), B=Get bShowPackageCallLogging()) -> (ReturnValue)
    Key=ItemSaleVerif
   )
  ]}
 GetLootPackageIdFromTierData {nodes [
   n1 = GetLootPackageIdFromTierData!() -> (then, TierGroupNames:Array of Names, LootTierDataTable:Data Table Object Reference) {is=title K2Node_FunctionEntry}
   > n2 = GetDataTableRowNames(Table=@n1.LootTierDataTable)
   > n3 = For Each Loop(Array=@n2)
   Return Node(execute=@n3.Completed)
   > n4 = Get Data Table Row(DataTable=@n1.LootTierDataTable, RowName=@n3) -> (ReturnValue:Fort Loot Tier Data Structure)
   n5 = Break Fort Loot Tier Data!(FortLootTierData:"Fort Loot Tier Data Structure (by ref)"=@n4) -> (
    TierGroup:Name
    LootPackage:Name
    NumLootPackageDropsRange:Vector 2D Structure
    RequiredTagQuery:Gameplay Tag Query Structure
    bAvoidDuplicateLootDrops:bool=~
   ) {is=title K2Node_BreakStruct}
   > n6 = Branch(Condition=Contains Item(TargetArray=@n1, ItemToFind=@n5))
   > n7 = Add Unique(TargetArray:Array of Names=Get LootPackageIdTrackingArray(), NewItem:"Name (by ref)"=@n5.LootPackage)
   > Branch(Condition="integer >= integer"(A=@n7))
   > PrintString(
    InString=Append(A="Tracking Package Id: ", B="To String (Name)"(InName=@n5.LootPackage), C=" from: ", D="To String (Name)"(InName=@n5))
    bPrintToScreen=false
    bPrintToLog=OR Boolean!(A=Get bShowAllLogging(), B=Get bShowLootPackageIdLogging() -> (bShowLootPackageIdLogging:bool), C:bool=Get bShowAllSuccessLogging()) -> (ReturnValue)
    Key=ItemSaleVerif
   )
   PrintString(
    execute=@n6.else
    InString=Append(
     A="TierGroup: "
     B="To String (Name)"(InName=@n5)
     C=" is not being tracked, not tracking the: "
     D="To String (Name)"(InName=@n5.LootPackage)
    )
    bPrintToScreen=false
    bPrintToLog=OR Boolean!(A=Get bShowAllLogging(), B=Get bShowLootPackageIdLogging()) -> (ReturnValue)
    Key=ItemSaleVerif
   )
  ]}
 GetTierGroupReferencesFromSalesTables {nodes [
   n1 = GetTierGroupReferencesFromSalesTables!() -> (then, SalesDataTable:Data Table Object Reference) {is=title K2Node_FunctionEntry}
   > n2 = GetDataTableRowNames(Table=@n1)
   > n3 = For Each Loop(Array=@n2)
   Return Node(execute=@n3.Completed)
   > n4 = Get Data Table Row(DataTable=@n1, RowName=@n3) -> (ReturnValue:NPCSale Inventory Row Structure)
   n5 = Break NPCSale Inventory Row!(NPCSaleInventoryRow:"NPCSale Inventory Row Structure (by ref)"=@n4) -> (LootTier:Name) {is=title K2Node_BreakStruct}
   > n6 = Add Unique(TargetArray=Get TierGroupTrackingArray(), NewItem=@n5)
   > Branch(Condition="integer >= integer"(A=@n6))
   > PrintString(
    InString=Append(A="Tracking Tier Group: ", B="To String (Name)"(InName=@n5), C=" from: ", D=GetDisplayName(Object=@n1))
    bPrintToScreen=false
    bPrintToLog=OR Boolean!(A=Get bShowAllLogging(), B=Get bShowTierGroupLogging() -> (bShowTierGroupLogging:bool), C:bool=Get bShowAllSuccessLogging()) -> (ReturnValue)
    Key=ItemSaleVerif
   )
  ]}
]
variables [
 bShowAllLogging {type=boolean}
 bShowAllSuccessLogging {type=boolean}
 bShowItemListLogging {type=boolean}
 bShowLootPackageIdLogging {type=boolean}
 bShowPackageCallLogging {type=boolean}
 bShowTierGroupLogging {type=boolean}
 CashPricingItemDef {type=Object, default=null}
 CashPricingTrackingArray {type="array<string>", default []}
 CurrentNpcSales {type=DataTable, default=null}
 CurrentStationSales {type=DataTable, default=null}
 LootCurrentSeasonLootPackageData {type=DataTable, default=null}
 LootCurrentSeasonLootTierData {type=DataTable, default=null}
 LootPackageCallTrackingArray {type="array<string>", default []}
 LootPackageIdTrackingArray {type="array<string>", default []}
 PackageItemListTrackingArray {type="array<string>", default []}
 TierGroupTrackingArray {type="array<string>", default []}
]
