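# Conversation database for the malfunctioning vending machine NPC, which offers a random RollTheDice trade in glitchy speech.
# A greeting leads to a BUY choice, gated on the service, then ACCEPT or BACK.
# Accepting gives a gift, plays service effects, applies a data-driven effect and repeats.
# Contextual lines react to the machine's pitch and roll, crown wearers, recent purchases and gift tags, with priority favouring some special cases.
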ConversationDatabase /NPCLibrary/NPCs/NPCLibrary_Conversation_VendingMachine_Malfunctioning
entry_points [{tag=Conversation.Entry.VendingMachine.Malfunctioning, destinations [#1]}]
nodes [
 ConversationEntryPointNode_7 {
  id=#1
  class=ConversationEntryPointNode
  outputs [#2]
  properties {entryTag {tagName=Conversation.Entry.VendingMachine.Malfunctioning}}
 }
 Speech_C_1 {
  id=#2
  class=Speech_C
  outputs [#3]
  properties {
   nodeName=""
   generalConfig {contextualMessages [
     {
      message="<b>*glitch*</> A-- a TrADe? A rando-- <b>*glitch*</> tr4d3?"
      contextRequirements []
      priority {value=0, curve {curveTable=null, rowName=None}}
     }
     {message="y0u'rE... <b>*glitch*</>", contextRequirements [], priority {value=0, curve {curveTable=null, rowName=None}}}
     {message=AaaAaA4ah..., contextRequirements [], priority {value=0, curve {curveTable=null, rowName=None}}}
     {message="be3e33ee3eePPP>...", contextRequirements [], priority {value=0, curve {curveTable=null, rowName=None}}}
     {message="<b>*robo-toot*</>", contextRequirements [], priority {value=0, curve {curveTable=null, rowName=None}}}
     {message="hiiIIIIIIIII <b>*BEEEEEEEEEP*</>", contextRequirements [], priority {value=0, curve {curveTable=null, rowName=None}}}
     {
      message="I... I'm fixed? You fixed me! I can't belie-- *glitch* 00000000h n00000..."
      contextRequirements [{
        participantId {tagName=Conversation.Participant.NPC}
        requirement {
         class=FortControllerRequirement_CheckPitchYawRoll
         properties {
          bInverseCondition=true
          pitchErrorCheck {bEnabled=true, toleratedErrorInDegrees=45}
          rollErrorCheck {bEnabled=true, toleratedErrorInDegrees=45}
         }
        }
       }]
      priority {curve {curveTable=null, rowName=None}}
     }
     {
      message="w...wHy... :("
      contextRequirements [{
        participantId {tagName=Conversation.Participant.NPC}
        requirement {
         class=FortControllerRequirement_CheckPitchYawRoll
         properties {
          bInverseCondition=true
          pitchErrorCheck {bEnabled=true, toleratedErrorInDegrees=45}
          rollErrorCheck {bEnabled=true, toleratedErrorInDegrees=45}
         }
        }
       }]
      priority {curve {curveTable=null, rowName=None}}
     }
     {
      message=R0yalTY...
      contextRequirements [{requirement {class=FortControllerRequirement_HasOwnedTag, properties {ownedTag {tagName=Granted.Abilities.VictoryCrown.CrownWearer}}}}]
      priority {value=2, curve {curveTable=null, rowName=None}}
     }
    ]}
  }
 }
 Speech_C_2 {
  id=#3
  class=Speech_C
  outputs [#4, #5]
  properties {
   softTaskIcon=/Game/Athena/HUD/Interaction/T_UI_VendingMachineRandom
   generalConfig {contextualMessages [
     {
      message=H3Lp... mE...
      contextRequirements [{requirement {class=FortControllerRequirement_HasOwnedTag, properties {ownedTag {tagName=Conversation.Context.ServiceJustPurchased}}}}]
      priority {curve {curveTable=null, rowName=None}}
     }
     {
      message="<b>*mechanical burping sounds*</>"
      contextRequirements [{requirement {class=FortControllerRequirement_HasOwnedTag, properties {ownedTag {tagName=Conversation.Context.ServiceJustPurchased}}}}]
      priority {curve {curveTable=null, rowName=None}}
     }
     {
      message="*sQ@wK!!*"
      contextRequirements [
       {requirement {class=FortControllerRequirement_HasOwnedTag, properties {ownedTag {tagName=Conversation.Context.ServiceJustPurchased}}}}
       {requirement {class=FortControllerRequirement_HasOwnedTag, properties {ownedTag {tagName=Athena.Quests.TandemGift.Nug}}}}
      ]
      priority {value=2, curve {curveTable=null, rowName=None}}
     }
     {
      message="<b>*glitch*</> Activ4ting self-cleaN m0de - extrem3 setTinG..."
      contextRequirements [
       {requirement {class=FortControllerRequirement_HasOwnedTag, properties {ownedTag {tagName=Conversation.Context.ServiceJustPurchased}}}}
       {requirement {class=FortControllerRequirement_HasOwnedTag, properties {ownedTag {tagName=Athena.Quests.TandemGift.Mobility}}}}
      ]
      priority {value=2, curve {curveTable=null, rowName=None}}
     }
     {
      message="Tr@de... complet3..."
      contextRequirements [{requirement {class=FortControllerRequirement_HasOwnedTag, properties {ownedTag {tagName=Conversation.Context.ServiceJustPurchased}}}}]
      priority {curve {curveTable=null, rowName=None}}
     }
     {
      message="<b>*CLUNK*</>"
      contextRequirements [{requirement {class=FortControllerRequirement_HasOwnedTag, properties {ownedTag {tagName=Conversation.Context.ServiceJustPurchased}}}}]
      priority {curve {curveTable=null, rowName=None}}
     }
     {
      message=i tHi1ink my bi1NarY c0D3 g0t a 2 iN it...
      contextRequirements [{requirement {class=FortControllerRequirement_HasOwnedTag, properties {ownedTag {tagName=Conversation.Context.ServiceJustPurchased}}}}]
      priority {curve {curveTable=null, rowName=None}}
     }
     {
      message=pl3aAse... c4LL a me3ndinG maCHine...
      contextRequirements [{requirement {class=FortControllerRequirement_HasOwnedTag, properties {ownedTag {tagName=Conversation.Context.ServiceJustPurchased}}}}]
      priority {curve {curveTable=null, rowName=None}}
     }
     {message=01101000 01100101 01101100 01110000, contextRequirements [], priority {value=0, curve {curveTable=null, rowName=None}}}
     {message=I d0n't.... fEEL so g00d..., contextRequirements [], priority {value=0, curve {curveTable=null, rowName=None}}}
     {message=baRs... Mor3 barS...??, contextRequirements [], priority {value=0, curve {curveTable=null, rowName=None}}}
     {message=juSt taAke itT..., contextRequirements [], priority {value=0, curve {curveTable=null, rowName=None}}}
     {message="o0okaAa4y... :(", contextRequirements [], priority {value=0, curve {curveTable=null, rowName=None}}}
     {message=TRADE. TRADE. TRADE. TRADE. TRADE., contextRequirements [], priority {value=0, curve {curveTable=null, rowName=None}}}
    ]}
  }
  sub_nodes [
   FortConversationChoiceWithPreviewNode_0 {class=FortConversationChoiceWithPreviewNode, properties {defaultChoiceDisplayText=BUY}}
   FortConversationRequirement_HasService_1 {class=FortConversationRequirement_HasService, properties {serviceTag {tagName=Tandem.Service.RollTheDice}}}
  ]
 }
 RollTheDice_C_1 {
  id=#5
  class=RollTheDice_C
  outputs [#6]
  properties {
   choiceSound=/Game/Athena/Sounds/NPCs/Conversation/Conversation_Buy_Cue
   effectRecipientConfigs [{recipient=Player, effects [{class=FortControllerEffect_NPCGiveGift}]}]
  }
  sub_nodes [
   ConversationChoiceNode_0 {class=ConversationChoiceNode, properties {defaultChoiceDisplayText=ACCEPT}}
   SideEffect_PlayServiceUsedEffects_C_1 {class=SideEffect_PlayServiceUsedEffects_C}
  ]
 }
 FortConversationTaskNode_ApplyDataDrivenEffect_2 {id=#6, class=FortConversationTaskNode_ApplyDataDrivenEffect, outputs [#7]}
 Repeat_C_0 {id=#7, class=Repeat_C}
 Back_C_1 {
  id=#4
  class=Back_C
  sub_nodes [ConversationChoiceNode_5 {class=ConversationChoiceNode, properties {defaultChoiceDisplayText=BACK}}]
 }
]
