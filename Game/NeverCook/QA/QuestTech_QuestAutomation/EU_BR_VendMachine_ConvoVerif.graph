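# Editor utility for QA that extracts vending machine conversation, sale speech and emote dialog text for verification.
# Run loops in turn over conversation databases, sales message collections and test-map machine actors, force-loading and casting each soft reference.
# Inherited EUO Item Sale Verification functions extract the messages and build output tables from each machine's dance and damage dialog.
# GatherAllDamageResponses flattens light, medium and heavy responses into one map keyed by severity and material style.
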EditorUtilityBlueprint /Game/NeverCook/QA/QuestTech_QuestAutomation/EU_BR_VendMachine_ConvoVerif
graphs [
 EventGraph {nodes [
   Event Run()
   > n1 = For Each Loop(Array:Array of Conversation Database Soft Object References=Get VendingMachine_ConvoDatabase() -> (VendingMachine_ConvoDatabase:Array of Conversation Database Soft Object References)) -> (Array Element:Conversation Database Soft Object Reference)
   n2 = For Each Loop(
    Exec=@n1.Completed
    Array:Array of Contextual Message Config Collection Soft Object References=Get DA_Speech_VendingMachine_Sales() -> (DA_Speech_VendingMachine_Sales:Array of Contextual Message Config Collection Soft Object References)
   ) -> (Array Element:Contextual Message Config Collection Soft Object Reference)
   n3 = For Each Loop(
    Exec=@n2.Completed
    Array:Array of BP Station Prop Vending Machine Emote Base Soft Object References=Get BP_StationProps_VM_Assault() -> (BP_StationProps_VM_Assault:Array of BP Station Prop Vending Machine Emote Base Soft Object References)
   ) -> (Array Element:BP Station Prop Vending Machine Emote Base Soft Object Reference)
   > n4 = ForceLoadAssetFromSoftObjectPtr!(execute, self:EUO Item Sale Verification Object Reference, ObjectToLoad:Soft Object Reference=@n3) -> (then, ReturnValue:Object Reference)
   > n5 = Cast To BP_StationProp_VendingMachine_Emote_Base!(execute, Object=@n4) -> (then, CastFailed, AsBP Station Prop Vending Machine Emote Base:BP Station Prop Vending Machine Emote Base Object Reference) {is=title K2Node_DynamicCast}
   > n6 = GetDialogInfo!(execute, self:BP Station Prop Vending Machine Emote Base Object Reference=@n5) -> (
    then
    DanceMessage:Text
    PlayerDanceMatStyleName:Name=None
    DanceMatStyleMap:Map of Fort Montage Item Definition Base Soft Object References to Names
    OnPlayerDanceMessages:Map of Fort Montage Item Definition Base Soft Object References to Texts
    DamageResponses_Light:Array of Vending Damage Response Structures
    DamageResponses_Med:Array of Vending Damage Response Structures
    DamageResponses_Heavy:Array of Vending Damage Response Structures
   )
   > n7 = GatherAllDamageResponses!(
    execute
    self
    DmgResponseLight:Array of Vending Damage Response Structures=@n6.DamageResponses_Light
    DmgResponseMed:Array of Vending Damage Response Structures=@n6.DamageResponses_Med
    DmgResponseHeavy:Array of Vending Damage Response Structures=@n6.DamageResponses_Heavy
   ) -> (then, DmgResponsesMapOutput:Map of Names to Texts)
   > CreateOutputTableFromVendingMachineBlueprints!(
    execute
    self:EUO Item Sale Verification Object Reference
    VmName:String=GetDisplayName(Object=@n5)
    OnDanceMsg:String="To String (Text)"(InText=@n6)
    OnDanceStyleMat:String="To String (Name)"(InName=@n6.PlayerDanceMatStyleName)
    DanceMatStyleMap:Map of Fort Montage Item Definition Base Soft Object References to Names=@n6.DanceMatStyleMap
    OnPlayerDanceMsgs:Map of Fort Montage Item Definition Base Soft Object References to Texts=@n6.OnPlayerDanceMessages
    DamageResponses:Map of Names to Texts=@n7
   ) -> (then)
   PrintString(execute=@n5.CastFailed, InString=ERROR: Cast failed...)
   n8 = ForceLoadAssetFromSoftObjectPtr(execute=@n2.LoopBody, ObjectToLoad=@n2)
   > n9 = Cast To ContextualMessageConfigCollection!(execute, Object=@n8) -> (then, CastFailed, AsContextual Message Config Collection:Contextual Message Config Collection Object Reference) {is=title K2Node_DynamicCast}
   > ExtractInfoFromContextualMessageConfigCollection!(execute, self:EUO Item Sale Verification Object Reference, InDataAssetConfig:Contextual Message Config Collection Object Reference=@n9) -> (then)
   PrintString(execute=@n9.CastFailed, InString=Failed to convert to ContextualMessageConfigCollection Object.)
   n10 = ForceLoadAssetFromSoftObjectPtr(execute=@n1.LoopBody, ObjectToLoad=@n1)
   > n11 = Cast To ConversationDatabase!(execute, Object=@n10) -> (then, CastFailed, AsConversation Database:Conversation Database Object Reference) {is=title K2Node_DynamicCast}
   > ExtractContextualMessagesFromSpeechInConvoDB!(execute, self:EUO Item Sale Verification Object Reference, InConvoDB:Conversation Database Object Reference=@n11) -> (then)
   PrintString(execute=@n11.CastFailed, InString=Failed to convert to ConversationDatabase Object.)
  ]}
 GatherAllDamageResponses {nodes [
   n1 = GatherAllDamageResponses!() -> (
    then
    DmgResponseLight:Array of Vending Damage Response Structures
    DmgResponseMed:Array of Vending Damage Response Structures
    DmgResponseHeavy:Array of Vending Damage Response Structures
   ) {is=title K2Node_FunctionEntry}
   > n2 = For Each Loop(Array:Array of Vending Damage Response Structures=@n1) -> (Array Element:Vending Damage Response Structure)
   n3 = For Each Loop(Exec=@n2.Completed, Array=@n1.DmgResponseMed)
   n4 = For Each Loop(Exec=@n3.Completed, Array=@n1.DmgResponseHeavy)
   Return Node!(execute=@n4.Completed, DmgResponsesMapOutput:Map of Names to Texts=Get OutputMap() -> (OutputMap:Map of Names to Texts))
   n5 = Break Vending Damage Response!(VendingDamageResponse:"Vending Damage Response Structure (by ref)"=@n4) -> (DamageMaterialStyle_5_#1:Name, DamageMessage_7_#2:Text) {is=title K2Node_BreakStruct}
   n6 = "To String (Name)"(InName=@n5)
   n7 = Get OutputMap()
   > Add!(
    execute
    TargetMap:Map of Names to Texts=@n7
    Key:"Name (by ref)"=String To Name(InString=Append(A="Heavy | ", B=@n6))
    Value:"Text (by ref)"=@n5.DamageMessage_7_#2
   ) -> (then) {is=title K2Node_CallFunction}
   n8 = Break Vending Damage Response(VendingDamageResponse=@n3)
   n9 = "To String (Name)"(InName=@n8)
   Add(
    execute=@n3.LoopBody
    TargetMap=@n7
    Key=String To Name(InString=Append(A="Med | ", B=@n9))
    Value=@n8.DamageMessage_7_#2
   )
   n10 = Break Vending Damage Response(VendingDamageResponse=@n2)
   n11 = "To String (Name)"(InName=@n10)
   Add(
    execute=@n2.LoopBody
    TargetMap=@n7
    Key=String To Name(InString=Append(A="Light | ", B=@n11))
    Value=@n10.DamageMessage_7_#2
   )
  ]}
]
variables [
 BP_StationProps_VM_Assault {
  type="array<BP_StationProp_VendingMachine_Emote_Base_C>"
  default [
   /Game/Athena/Maps/Test/Quests/NPC_Services_Testing:PersistentLevel.BP_StationProp_VendingMachine_WeaponsSniper_0
   &NPC_Services_Testing:PersistentLevel.BP_StationProp_VendingMachine_WeaponsAR_0
   &NPC_Services_Testing:PersistentLevel.BP_StationProp_VendingMachine_WeaponsPistol_0
   &NPC_Services_Testing:PersistentLevel.BP_StationProp_VendingMachine_WeaponsSMG_0
  ]
 }
 DA_Speech_VendingMachine_Sales {type="array<ContextualMessageConfigCollection>", default [/NPCLibrary/NPCs/DA_Speech_VendingMachine_Sales]}
 VendingMachine_ConvoDatabase {
  type="array<ConversationDatabase>"
  default [
   /NPCLibrary/NPCs/NPCLibrary_Conversation_VendingMachine_Weapons
   /NPCLibrary/NPCs/NPCLibrary_Conversation_VendingMachine_Medical
   /NPCLibrary/NPCs/NPCLibrary_Conversation_VendingMachine_Malfunctioning
  ]
 }
]
